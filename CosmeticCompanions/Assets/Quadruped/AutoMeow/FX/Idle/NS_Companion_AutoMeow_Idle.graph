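# Idle Niagara effect for the AutoMeow quadruped cosmetic companion: soft glow sprites tinted by the user eye color.
# A single Glow emitter bursts three persistent sprites, one per spawn group, each with its own size and camera offset.
# Each sprite follows its indexed socket with a small offset and aligns to the socket's rotation.
# Color multiplies the eye tint by a system time-of-day color, with alpha differing between gameplay and front end.
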NiagaraSystem /CosmeticCompanions/Assets/Quadruped/AutoMeow/FX/Idle/NS_Companion_AutoMeow_Idle
dependencies {
 usedDataInterfaces [&NiagaraDataInterfaceSocketReader, &NiagaraDataInterfaceColorCurve]
 usedDynamicInputs [&FNTimeOfDayScalar, &ColorFromCurve]
 usedModules [
  &self:SystemSpawnScript.SystemScriptSource.SystemScriptGraph.NiagaraNodeAssignment_1.SetVariables_#1
  &SystemState
  &self:SystemSpawnScript.SystemScriptSource.SystemScriptGraph.NiagaraNodeAssignment_3.SetVariables_#2
 ]
}
emitters {Glow {
  emitterUpdateScript [
   &EmitterState(
    Inactive Response=::NewEnumerator1
    Life Cycle Mode=::NewEnumerator1
    Loop Behavior=::NewEnumerator1
    Loop Duration Mode=::NewEnumerator1
    UseLoopDelay=false
   )
   &SpawnBurst_Instantaneous(Spawn Count=3, Use Spawn Probability=false, UseLoopCountLimit=false)
  ]
  particleSpawnScript [
   &FN_SetSpawnGroup(Number of Spawn Groups=3)
   &InitializeParticle(
    Position Mode=::NewEnumerator0
    Sprite Size=&Multiply_Vector2D_ByFloat(
     Float=15
     Vector2D=&MergeHelper_Vector2Switch(0=<0.7 1.2>, 1=<0.7 1.2>, 2=<1 1.4>, 3=?, 4=?, Index=&SpawnGroup(), Num=::NewEnumerator1)
    )
    Sprite Size Mode=::NewEnumerator4
    Write Material Random=true
   )
   &CameraOffset(Camera Offset Amount=&MergeHelper_FloatSwitch(0=1, 1=1, 2=2, 3=?, Index=&SpawnGroup(), Num=::NewEnumerator1))
   /SetVariables(Particles.VisibilityTag=&MergeHelper_IntSwitch(0=0, 1=0, 2=0, 3=1, 4=2, 5=2, 6=2, Index=&SpawnGroup(), Num=::NewEnumerator5)) {isSetParameters=true}
  ]
  particleUpdateScript [
   &ParticleState(
    Kill Particles When Lifetime Has Elapsed=false
    Let Infinitely Lived Particles Die When Emitter Deactivates=true
    Loop Particles Lifetime=true
   )
   &SocketLocation(
    SamplePosOffset=true
    Sampled Position Offset=&MergeHelper_VectorSwitch(0=<0 1 0>, 1=<0 -1 0>, 2=<-1.8 0 0>, 3=?, Index=&SpawnGroup(), Num=::NewEnumerator1)
    Socket Index=&SpawnGroup()
    Socket Reader=$System.Socket Reader
    Socket Sampling Mode=::NewEnumerator3
    VelocitySpeedLimit=false
   )
   &DynamicMaterialParameters(
    Index 0 Param 1=1
    Index 0 Param 3=5
    Module.Index0Param0Write=true
    Module.Index0Param1Write=false
    Module.Index0Param2Write=false
    Module.Index0Param3Write=false
   )
   &ScaleColor(
    Scale Alpha=&FN_FrontEndFloatModifier(Gameplay=0.07, Front End=0.045)
    Scale RGB=&MakeVectorFromLinearColorRGB(Color=&MultiplyLinearColor(
      Linear Color=&Multiply_LinearColorByFloat(Linear Color=$User.EyeTintColor, Float=2)
      Scale Factor=&MakeVector4_FromLinearColor(LinearColor=$System.TOD Color)
     ))
    ScaleA=true
    ScaleRGB=true
   )
   /Game/Sprite/FN_SpriteAlignmentAndFacing(SpriteAlignment:Vector3f=&RotateVectorByQuaternion(Vector To Rotate=<1 0 0>, Quaternion=$Output.SocketLocation.SampledRotation))
  ]
  rendererClasses [&NiagaraSpriteRendererProperties]
  renderers [{rendererClass=&NiagaraSpriteRendererProperties}]
 }}
summary {
 systemName=NS_Companion_AutoMeow_Idle
 emitters [{emitterName=Glow, rendererClasses [&NiagaraSpriteRendererProperties]}]
 userVariables [
  User.bIsFrontEnd {type=NiagaraBool}
  User.bIsFrontEndPreview {type=NiagaraBool}
  User.EyeTintColor {type=LinearColor, description="", defaultValue {struct=LinearColor, value=<rgba 1 1 1 1>}}
 ]
}
